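# Physics asset for the Moth cape cosmetic back accessory, setting up physics bodies and joints for its left and right wings.
# A small sphere at backpack_root anchors two mirrored chains of four capsule bodies, one per wing bone.
# Constraints let each wing segment bend mainly on one swing axis with almost no side swing or twist.
# The root joints are tighter, limited to that single axis.
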PhysicsAsset /BRCosmetics/Accessories/FORT_Capes/Cape_M_MED_Moth/Meshes/M_MED_Moth_Cape_Physics_02
bodies {
 backpack_root [{shapeType=Sphere, radius=2}]
 wing_l_01 [{shapeType=Capsule, length=5, radius=5, center=<5 0 0>}]
 wing_l_02 [{shapeType=Capsule, length=5, radius=5, center=<1 0 0>}]
 wing_l_03 [{shapeType=Capsule, length=5, radius=5, center=<1 0 0>}]
 wing_l_04 [{shapeType=Capsule, length=5, radius=5, center=<1 0 0>}]
 wing_r_01 [{shapeType=Capsule, length=5, radius=5, center=<5 0 0>}]
 wing_r_02 [{shapeType=Capsule, length=5, radius=5, center=<1 0 0>}]
 wing_r_03 [{shapeType=Capsule, length=5, radius=5, center=<1 0 0>}]
 wing_r_04 [{shapeType=Capsule, length=5, radius=5, center=<1 0 0>}]
}
constraints [
 {bone1Name=wing_r_04, bone2Name=wing_r_03, swing1LimitDegrees=45, swing2LimitDegrees=2, twistLimitDegrees=2}
 {bone1Name=wing_r_03, bone2Name=wing_r_02, swing1LimitDegrees=45, swing2LimitDegrees=2, twistLimitDegrees=2}
 {bone1Name=wing_r_02, bone2Name=wing_r_01, swing1LimitDegrees=45, swing2LimitDegrees=2, twistLimitDegrees=2}
 {bone1Name=wing_l_04, bone2Name=wing_l_03, swing1LimitDegrees=45, swing2LimitDegrees=2, twistLimitDegrees=2}
 {bone1Name=wing_l_03, bone2Name=wing_l_02, swing1LimitDegrees=45, swing2LimitDegrees=2, twistLimitDegrees=2}
 {bone1Name=wing_l_02, bone2Name=wing_l_01, swing1LimitDegrees=45, swing2LimitDegrees=2, twistLimitDegrees=2}
 {bone1Name=wing_l_01, bone2Name=backpack_root, swing1LimitDegrees=20, swing2LimitDegrees=0, twistLimitDegrees=0}
 {bone1Name=wing_r_01, bone2Name=backpack_root, swing1LimitDegrees=20, swing2LimitDegrees=0, twistLimitDegrees=0}
]
