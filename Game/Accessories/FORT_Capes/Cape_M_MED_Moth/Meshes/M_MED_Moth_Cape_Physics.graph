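# Physics asset for the medium male Moth cape accessory, giving its two wings physics bodies attached to the backpack root.
# A small sphere on backpack_root anchors one offset capsule on each of wing_l_01 and wing_r_01, with the left and right capsules sized differently.
# Each wing is constrained directly to the root with a wide first swing limit and tight second swing and twist limits, so it flaps mainly about one axis.
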PhysicsAsset /Game/Accessories/FORT_Capes/Cape_M_MED_Moth/Meshes/M_MED_Moth_Cape_Physics
bodies {
 backpack_root [{shapeType=Sphere, radius=2}]
 wing_l_01 [{shapeType=Capsule, length=14.5, radius=5, center=<35 -7 0>, rotation=<rot 0 -90 0>}]
 wing_r_01 [{shapeType=Capsule, length=17.5, radius=3, center=<35 -7 0>, rotation=<rot 0 -90 0>}]
}
constraints [
 {bone1Name=wing_l_01, bone2Name=backpack_root, swing1LimitDegrees=45, swing2LimitDegrees=5, twistLimitDegrees=2}
 {bone1Name=wing_r_01, bone2Name=backpack_root, swing1LimitDegrees=45, swing2LimitDegrees=5, twistLimitDegrees=2}
]
