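# Animation Blueprint for the Moth back-bling cape, simulating cloth-like motion with rigid body physics.
# The live chain runs a RigidBody node from reference pose in backpack_root space, then additively rotates that bone by yaw remapped from the owner mesh's spine bend.
# Skydive events swap normal gravity for a tiny upward value.
# A second RigidBody node, the only user of the skydive alpha, is left unconnected to the output.
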AnimBlueprint /BRCosmetics/Accessories/FORT_Capes/Cape_M_MED_Moth/Meshes/M_MED_Moth_Cape_AnimBP
target_skeleton=&/M_MED_Moth_Cape_Skeleton
graphs [
 AnimGraph {nodes [
   n1 = Get override_world_gravity() -> (override_world_gravity:Vector)
   RigidBody!(ComponentPose, Alpha=Get skydive_alpha() -> (skydive_alpha:double), OverrideWorldGravity:Vector=@n1, ExternalForce) -> (Pose) {
    bOverrideWorldGravity=true
    overridePhysicsAsset=/Game/Accessories/FORT_Capes/Cape_M_MED_Moth/Meshes/M_MED_Moth_Cape_Physics
    simulationSpace=BaseBoneSpace
    alphaBoolBlend {blendOption=Linear}
    baseBoneRef {boneName=backpack_root}
    componentLinearAccScale=<1 1 1>
    componentLinearVelScale=<1 1 1>
    overrideWorldGravity {z=1}
   }
   n2 = RigidBody(ComponentPose=Local To Component(LocalPose=Local Space Ref Pose()), OverrideWorldGravity=@n1) {
    bOverrideWorldGravity=true
    overridePhysicsAsset=&/M_MED_Moth_Cape_Physics_02
    simulationSpace=BaseBoneSpace
    alphaBoolBlend {blendOption=Linear}
    baseBoneRef {boneName=backpack_root}
    componentAppliedLinearAccClamp=<100 100 100>
    componentLinearAccScale=<1 1 1>
    componentLinearVelScale=<1 1 1>
    overrideWorldGravity {z=1}
   }
   n3 = "Transform (Modify) Bone - Bone: backpack_root"(ComponentPose=@n2, Alpha=1, Translation="0,0,0", Rotation=Get backpack_rotate(), Scale="1,1,1") {rotationMode=BMM_Additive, rotationSpace=BCS_BoneSpace, alphaBoolBlend {blendOption=Linear}, boneToModify {boneName=backpack_root}}
   Output Pose(Result=Component To Local(ComponentPose=@n3)) {layerGroup=None}
   n4 = Get isSD() -> (isSD:bool)
   SelectVector(A="0, 0,1", B="0, 0,-980", bPickA=@n4)
   SelectFloat(A=0.1, B=1, bPickA=@n4)
  ]}
 EventGraph {nodes [
   Event BlueprintUpdateAnimation()
   n1 = GetOwnerSkeletalMeshComponent()
   n2 = BreakRotator(InRot=BreakTransform(InTransform=GetDeltaTransformFromRefPose(self=@n1, BoneName=spine_05, BaseName=spine_01)).Rotation)
   > Set backpack_rotate(backpack_rotate=MakeRotator(Yaw=MapRangeClamped(Value=@n2.Yaw, InRangeA=12, InRangeB=-30, OutRangeA=8, OutRangeB=30)))
   n3 = Get isSD()
   > Set override_world_gravity(override_world_gravity:Vector=SelectVector(A="0, 0,1", B="0, 0,-980", bPickA=@n3)) -> (Output_Get:Vector)
   > Set skydive_alpha(skydive_alpha:double=SelectFloat(A=0.1, B=1, bPickA=@n3)) -> (Output_Get:double)
   Event OnBeginSkydiving()
   > Set isSD(isSD:bool=true) -> (Output_Get:bool)
   Event OnEndSkydiving()
   > Set isSD(isSD=false)
  ]}
]
variables [
 backpack_rotate {type=Rotator}
 isSD {type=boolean}
 override_world_gravity {type=Vector}
 skydive_alpha {type=number}
]
